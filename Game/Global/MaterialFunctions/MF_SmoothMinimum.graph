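# MF_SmoothMinimum is a material function that returns a smooth minimum of two scalars, x and y, with a distance input setting the blend width.
# It saturates half their difference divided by distance, offset by 0.5, to get a blend factor h.
# It lerps x to y by h, then subtracts distance × h × (1 − h), the standard polynomial smooth-min that rounds the crossover.
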MaterialFunction /Game/Global/MaterialFunctions/MF_SmoothMinimum
graph {nodes [
  n1 = FunctionInput() {bUsePreviewValueAsDefault=true, inputName=x, inputType=FunctionInput_Scalar}
  n2 = FunctionInput() {bUsePreviewValueAsDefault=true, inputName=y, inputType=FunctionInput_Scalar, sortPriority=1}
  n3 = Multiply(A=Subtract(A=@n1, B=@n2)) {constB=0.5}
  n4 = FunctionInput() {bUsePreviewValueAsDefault=true, inputName=distance, inputType=FunctionInput_Scalar, sortPriority=2}
  n5 = Saturate(None=Add(A=Divide(A=@n3, B=@n4)) {constB=0.5})
  n6 = OneMinus(None=@n5)
  n7 = Multiply(A=@n5, B=@n4)
  FunctionOutput(None=Subtract(A=LinearInterpolate(A=@n1, B=@n2, Alpha=@n5), B=Multiply(A=@n6, B=@n7)))
 ]}
