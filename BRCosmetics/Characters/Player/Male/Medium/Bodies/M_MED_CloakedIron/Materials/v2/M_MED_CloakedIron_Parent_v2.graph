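# Parent material for the CloakedIron medium male player body cosmetic (v2), built on the shared 2023 character pre-FX and post-FX material functions.
# The pre-FX output passes through a CloakedIron pattern animation function, masked by the RGB of a black-and-white pattern mask texture, before entering post-FX.
# It uses material attributes with shading model from expression and flags pixel animation.
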Material /BRCosmetics/Characters/Player/Male/Medium/Bodies/M_MED_CloakedIron/Materials/v2/M_MED_CloakedIron_Parent_v2
graph {
 properties {
  bHasPixelAnimation=true
  bUseMaterialAttributes=true
  bUsedWithClothing=true
  bUsedWithMeshDeformer=true
  bUsedWithMorphTargets=true
  bUsedWithNanite=true
  bUsedWithNiagaraMeshParticles=true
  bUsedWithSkeletalMesh=true
  pixelDepthOffsetMode=PDOM_Legacy
  shadingModel=MSM_FromMaterialExpression
  usedShadingModels=Default Lit | Subsurface | Clear Coat | Subsurface Profile | Cloth
 }
 nodes [
  n1 = TextureSample() {
   samplerType=SAMPLERTYPE_Masks
   texture=/BRCosmetics/Characters/Player/Male/Medium/Bodies/M_MED_CloakedIron/Textures/T_CloakedIron_BW_PatternMasks
  }
  n2 = MaterialFunctionCall() {materialFunction=&MF_CharacterParent_2023_PreFX}
  MaterialFunctionCall(
   Material=MaterialFunctionCall!(Mask=@n1.RGB, In=@n2.Result) {materialFunction=&/MF_CloakedIron_PatternAnim_v2}.Result
   M=@n2.M
   Pre Skinned Local Position=@n2.M
   Distance Blend=@n2.M
   Fresnel=@n2.M
  ) {materialFunction=&MF_CharacterParent_2023_PostFX}
 ]
}
